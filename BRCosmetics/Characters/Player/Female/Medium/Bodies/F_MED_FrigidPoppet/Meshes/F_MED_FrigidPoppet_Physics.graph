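# Physics asset for the F_MED_FrigidPoppet medium female player body cosmetic, defining its simplified collision and joint setup.
# Single capsules cover the pelvis, three spine bones, arms, hands, legs and feet, with no head or neck body.
# Constraints chain limbs to spine_05 and pelvis with uniform swing and twist limit angles.
# The spine_01 body has no constraint, as spine_04 links directly to the pelvis.
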PhysicsAsset /BRCosmetics/Characters/Player/Female/Medium/Bodies/F_MED_FrigidPoppet/Meshes/F_MED_FrigidPoppet_Physics
bodies {
 calf_l [{shapeType=Capsule, length=35.4084, radius=8.6466, center=<-25.586 -1.2718 -1.0198>, rotation=<rot 0 86.8993 -89.3637>}]
 calf_r [{shapeType=Capsule, length=35.4084, radius=8.6466, center=<25.586 1.2718 1.0198>, rotation=<rot 0 86.8993 90.6363>}]
 foot_l [{shapeType=Capsule, length=15.5614, radius=8.4479, center=<-3.8831 6.8542 -0.3834>, rotation=<rot -0.2383 -179.6037 -90.7401>}]
 foot_r [{shapeType=Capsule, length=15.5614, radius=8.4479, center=<3.8831 -6.8542 0.3834>, rotation=<rot -0.2383 -179.6037 89.2599>}]
 hand_l [{shapeType=Capsule, length=8.6156, radius=5.7707, center=<8.1322 -2.6261 -0.0688>, rotation=<rot 0 83.9027 -70.1169>}]
 hand_r [{shapeType=Capsule, length=8.6156, radius=5.7707, center=<-8.1322 2.6261 0.0688>, rotation=<rot 0 83.9027 -70.1169>}]
 lowerarm_l [{shapeType=Capsule, length=26.1607, radius=6.2498, center=<15.0788 0.2772 -0.8631>, rotation=<rot -21.4684 -90 95.6135>}]
 lowerarm_r [{shapeType=Capsule, length=26.1607, radius=6.2498, center=<-15.0788 -0.2772 0.8631>, rotation=<rot -21.4684 -90 -84.3865>}]
 pelvis [{shapeType=Capsule, length=10.6185, radius=15.5636, center=<-8.9323 -1.9106 0>}]
 spine_01 [{shapeType=Capsule, length=12.6185, radius=13.5636, center=<-1.9569 -0.6159 0>}]
 spine_04 [{shapeType=Capsule, length=16.4651, radius=15.4902, center=<0 3.4378 0>, rotation=<rot 0 90 -90>}]
 spine_05 [{shapeType=Capsule, length=16.9797, radius=13.2608, center=<5.497 -1.2692 -0.5376>}]
 thigh_l [{shapeType=Capsule, length=30.1632, radius=10.1667, center=<-22.4673 1.1267 -0.9448>, rotation=<rot 0 85.2774 -88.5772>}]
 thigh_r [{shapeType=Capsule, length=30.1632, radius=10.1667, center=<22.4673 -1.1267 0.9448>, rotation=<rot 0 85.2774 91.4228>}]
 upperarm_l [{shapeType=Capsule, length=17.8862, radius=6.3328, center=<12.6772 0.0702 -0.7322>, rotation=<rot 0 87.8347 90.0695>}]
 upperarm_r [{shapeType=Capsule, length=17.8862, radius=6.3328, center=<-12.6772 -0.0702 0.7322>, rotation=<rot 0 87.8347 -89.9305>}]
}
constraints [
 {bone1Name=spine_04, bone2Name=pelvis, swing1LimitDegrees=45, swing2LimitDegrees=45, twistLimitDegrees=45}
 {bone1Name=spine_05, bone2Name=spine_04, swing1LimitDegrees=45, swing2LimitDegrees=45, twistLimitDegrees=45}
 {bone1Name=upperarm_l, bone2Name=spine_05, swing1LimitDegrees=45, swing2LimitDegrees=45, twistLimitDegrees=45}
 {bone1Name=lowerarm_l, bone2Name=upperarm_l, swing1LimitDegrees=45, swing2LimitDegrees=45, twistLimitDegrees=45}
 {bone1Name=hand_l, bone2Name=lowerarm_l, swing1LimitDegrees=45, swing2LimitDegrees=45, twistLimitDegrees=45}
 {bone1Name=upperarm_r, bone2Name=spine_05, swing1LimitDegrees=45, swing2LimitDegrees=45, twistLimitDegrees=45}
 {bone1Name=lowerarm_r, bone2Name=upperarm_r, swing1LimitDegrees=45, swing2LimitDegrees=45, twistLimitDegrees=45}
 {bone1Name=hand_r, bone2Name=lowerarm_r, swing1LimitDegrees=45, swing2LimitDegrees=45, twistLimitDegrees=45}
 {bone1Name=thigh_l, bone2Name=pelvis, swing1LimitDegrees=45, swing2LimitDegrees=45, twistLimitDegrees=45}
 {bone1Name=calf_l, bone2Name=thigh_l, swing1LimitDegrees=45, swing2LimitDegrees=45, twistLimitDegrees=45}
 {bone1Name=foot_l, bone2Name=calf_l, swing1LimitDegrees=45, swing2LimitDegrees=45, twistLimitDegrees=45}
 {bone1Name=thigh_r, bone2Name=pelvis, swing1LimitDegrees=45, swing2LimitDegrees=45, twistLimitDegrees=45}
 {bone1Name=calf_r, bone2Name=thigh_r, swing1LimitDegrees=45, swing2LimitDegrees=45, twistLimitDegrees=45}
 {bone1Name=foot_r, bone2Name=calf_r, swing1LimitDegrees=45, swing2LimitDegrees=45, twistLimitDegrees=45}
]
